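# Niagara system for the male BlueFlames skin cosmetic that renders blue flaming hair with two sprite emitters on the character's skeletal mesh.
# Flames_Hair spawns sprites in a flat disc at a mesh offset and drifts them with drag, animating color and material parameters.
# Decal_Hair re-pins glow sprites to the mesh every update at four partition-selected offsets, pulsing and fading them.
# Both map User.bisHeadHidden to a particle VisibilityTag.
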NiagaraSystem /BRCosmetics/Athena/Effects/Particles/Skins/BlueFlames/NS_BlueFlames_Male_Hair
dependencies {
 usedDataInterfaces [
  &NiagaraDataInterfaceColorCurve
  &NiagaraDataInterfaceCurve
  &NiagaraDataInterfaceArrayFloat3
  &NiagaraDataInterfaceSkeletalMesh
 ]
 usedDynamicInputs []
 usedModules [
  &SystemState
  &self:SystemSpawnScript.SystemScriptSource.SystemScriptGraph.NiagaraNodeAssignment_0.SetVariables_#1
 ]
}
emitters {
 Decal_Hair {
  emitterUpdateScript [
   &EmitterState()
   &SpawnRate(SpawnRate=6, Use Spawn Probability=false)
  ]
  particleSpawnScript [
   &PartitionParticles(Number Of Partitions=4)
   &InitializeParticle(
    Color=<rgba 0 1.4562 4 0.5>
    Lifetime Max=1
    Lifetime Min=0.75
    Lifetime Mode=::NewEnumerator1
    Sprite Rotation Angle=&RandomRangeFloat(Minimum=-0.05, Maximum=0.05)
    Sprite Rotation Mode=::NewEnumerator1
    Sprite Size Mode=::NewEnumerator1
    Uniform Sprite Size Max=25
    Uniform Sprite Size Min=20
    UsePositionOffset=false
    Write Material Random=true
   )
   &CameraOffset(Camera Offset Amount=10)
  ]
  particleUpdateScript [
   &ParticleState()
   &SkeletalMeshLocation(
    Bone Sampling Mode=::NewEnumerator4
    MaskSpawnGroup=false
    Sampled Position Offset=&SelectVectorFromArray(
     Direct Array Index=$Particles.PartitionParticles.Partition
     Vector Selection Array=DataInterface() {propertyValues {FloatData [<17 -4 0>, <18 2.5 0>, <14 -2 0>, <16 -8 0>]}}
    )
    Skeletal Mesh=$System.Skeletal Mesh
   )
   &DynamicMaterialParameters(
    Index 0 Param 1=&FloatFromCurve(FloatCurve=DataInterface() {propertyValues {Curve {keys [<at 0 2>,<at 1 12>]}}})
    Module.Index0Float0Write=true
    Module.Index0Float1Write=false
    Module.Index0Float2Write=true
    Module.Index0Float3Write=false
   )
   &ScaleColor(
    Scale Alpha=&FloatFromCurve(FloatCurve=DataInterface() {propertyValues {Curve {keys [<at 0 1>,<at 1 0>]}}})
    Scale RGB=&MakeVectorFromLinearColorRGB(Color=&ColorFromCurve(ColorCurve=DataInterface() {propertyValues {
        AlphaCurve {keys []}
        BlueCurve {keys [<at 0 0.0953>,<at 0.3 25>,<at 1 0.0953>]}
        GreenCurve {keys [<at 0 0.0497>,<at 0.3 7.8202>,<at 1 0.0497>]}
        RedCurve {keys [<at 0 0.0343>,<at 0.3 0>,<at 1 0.0343>]}
       }}))
    ScaleA=true
    ScaleRGB=true
   )
   /SetVariables(Particles.VisibilityTag=&MakeCustomIntFromBool(Bool=$User.bisHeadHidden, True INT=1, False INT=0)) {isSetParameters=true, module=SetVariables_#2}
  ]
  rendererClasses [&NiagaraSpriteRendererProperties]
  renderers [{rendererClass=&NiagaraSpriteRendererProperties}]
 }
 Flames_Hair {
  emitterUpdateScript [
   &EmitterState()
   &SpawnRate(SpawnRate=6, Use Spawn Probability=false)
  ]
  particleSpawnScript [
   &InitializeParticle(
    Lifetime Max=1.5
    Lifetime Mode=::NewEnumerator1
    Sprite Size Max=<14 20>
    Sprite Size Min=<8 14>
    Sprite Size Mode=::NewEnumerator2
    UsePositionOffset=false
    Write Material Random=true
   )
   &AddVelocity(Velocity=&RandomRangeVector(Minimum=<14 0 0>, Maximum=<18 -10 1>))
   &SkeletalMeshLocation(Bone Sampling Mode=::NewEnumerator4, MaskSpawnGroup=false, Sampled Position Offset=<12 -1 0>, Skeletal Mesh=$System.Skeletal Mesh)
   &ShapeLocation(
    Cylinder Height=0
    Cylinder Radius=8.5
    MaskSpawnGroup=false
    Override Seed=false
    Shape Primitive=::NewEnumerator1
    Surface Only=false
    UseOwnerScale=true
   )
   &CameraOffset(Camera Offset Amount=8)
  ]
  particleUpdateScript [
   &ParticleState()
   &Drag(Use Linear Drag=true, Use Rotational Drag=true)
   &SolveForcesAndVelocity(Clamp Velocity=false, Limit Acceleration=false)
   &Color(Color=&ColorFromCurve(ColorCurve=DataInterface() {propertyValues {
       AlphaCurve {keys [
         {arriveTangent=-8.7552, interpMode=RCIM_Cubic, leaveTangent=-8.7552, tangentMode=RCTM_User, time=0, value=1}
         {arriveTangentWeight=0.0944, interpMode=RCIM_Cubic, leaveTangentWeight=0.1889, tangentMode=RCTM_User, time=0.25, value=0.3}
         <at 0.75 0.6>
         <at 1 1 RCIM_Cubic>
        ]}
       BlueCurve {keys [<at 0.4 0.1042 RCIM_Cubic>,<at 0.9 50 RCIM_Cubic>]}
       GreenCurve {keys [<at 0.4 0.0543 RCIM_Cubic>,<at 0.9 14.461 RCIM_Cubic>]}
       RedCurve {keys [<at 0.4 0.0375 RCIM_Cubic>,<at 0.9 0 RCIM_Cubic>]}
      }}))
   &DynamicMaterialParameters(
    Index 0 Param 1=&FloatFromCurve(FloatCurve=DataInterface() {propertyValues {Curve {keys [<at 0 -0.01>,<at 1 -0.6>]}}})
    Index 0 Param 2=&RandomRangeFloat(Minimum=0, Maximum=100)
    Index 0 Param 3=&FloatFromCurve(FloatCurve=DataInterface() {propertyValues {Curve {keys [<at 0.25 0.5>,<at 1 1>]}}})
    Module.Index0Float0Write=true
    Module.Index0Float1Write=true
    Module.Index0Float2Write=true
    Module.Index0Float3Write=false
   )
   /SetVariables(Particles.VisibilityTag=&MakeCustomIntFromBool(Bool=$User.bisHeadHidden, True INT=1, False INT=0)) {isSetParameters=true, module=SetVariables_#2}
  ]
  rendererClasses [&NiagaraSpriteRendererProperties]
  renderers [{rendererClass=&NiagaraSpriteRendererProperties}]
 }
}
summary {
 systemName=NS_BlueFlames_Male_Hair
 emitters [
  {emitterName=Flames_Hair, rendererClasses [&NiagaraSpriteRendererProperties]}
  {emitterName=Decal_Hair, rendererClasses [&NiagaraSpriteRendererProperties]}
 ]
 userVariables [User.bisHeadHidden {type=NiagaraBool}]
}
